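# Niagara system for the Illustrative vehicle boost cosmetic, emitting exhaust effects from filtered sockets on the vehicle's skeletal mesh.
# ConeBase draws mesh and sprite bursts, while Sprites throws particles backward, with spawn groups varying lifetime and motion.
# A wide ribbon trails each port, wrapped by two thinner ribbons using sine/cosine velocity and oppositely signed ring rotations to twist.
# Each emitter sets its own painted-variant palette, with secondary colors on ConeBase and Sprites.
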NiagaraSystem /VehicleCosmetics/Boosts/Boost_Illustrative/FX/NS_Boost_Illustrative
dependencies {
 usedDataInterfaces [
  &NiagaraDataInterfaceVectorCurve
  &NiagaraDataInterfaceCurve
  &NiagaraDataInterfaceArrayFloat2
  &NiagaraDataInterfaceArrayFloat
  &NiagaraDataInterfaceSkeletalMesh
 ]
 usedDynamicInputs []
 usedModules [
  &self:SystemSpawnScript.SystemScriptSource.SystemScriptGraph.NiagaraNodeAssignment_0.SetVariables_#1
  &SystemState
 ]
}
emitters {
 ConeBase {
  emitterUpdateScript [
   &EmitterState()
   &VC_FilteredSocketCount()
   &SpawnRate(SpawnRate=30, Use Spawn Probability=false)
   &SpawnRate(SpawnGroup=1, SpawnRate=20, Use Spawn Probability=false) {module=SpawnRate001}
   &SpawnRate(SpawnGroup=2, SpawnRate=20, Use Spawn Probability=false) {module=SpawnRate002}
  ]
  particleSpawnScript [
   /SetVariables(Particles.RenderTag:NiagaraInt32=&SpawnGroup()) {isSetParameters=true}
   &VC_PaintedVariantsOverride(
    Black=<rgba 0 0 0 1.5>
    BurntSienna=<rgba 0.2174 0.052 0.002 1>
    Cobalt=<rgba 0.0435 0.0925 5 1>
    Crimson=<rgba 0.1823 0.0028 0.0035 1.4>
    DefaultColor=<rgba 1 0.7992 0.0781 1>
    ForestGreen=<rgba 0.0169 0.2604 0 1.4>
    Grey=<rgba 0.03 0.03 0.03 1.5>
    Lime=<rgba 0.2943 0.5885 0 1>
    Pink=<rgba 0.6875 0.0967 0.4727 1>
    Purple=<rgba 0.159 0.0272 0.2292 1.2>
    R1 Cobalt:LinearColor=<rgba 0.0435 0.0925 5 1>
    R1 DefaultColor=<rgba 1 0.0354 0 1>
    Saffron=<rgba 0.6458 0.5415 0.0364 1.3>
    SkyBlue=<rgba 0 0.4055 0.6094 1>
    Use R1 Painted Overrides=true
   )
   &InitializeParticle(
    Color=$Particles.Color
    Lifetime=/VehicleCosmetics/FX/Niagara/Modules/MergeEmitters/VC_FloatBySpawnGroup(
     SpawnGroup or ExecIndex:NiagaraInt32=0
     CustomInt:NiagaraInt32=0
     MaximumSpawnGroups:NiagaraInt32=3
     Float SG 0:NiagaraFloat=&RandomRangeFloat(Minimum=0.5, Maximum=0.3)
     Float SG 1:NiagaraFloat=&RandomRangeFloat(Minimum=0.1, Maximum=0.4)
     Float SG 2:NiagaraFloat=&RandomRangeFloat(Minimum=0.1, Maximum=0.4)
     Float SG 3:NiagaraFloat=0
     Float SG 4:NiagaraFloat=0
     Float SG 5:NiagaraFloat=0
    )
    Mesh Scale Mode=::NewEnumerator3
    Mesh Uniform Scale=3
    Sprite Rotation Mode=::NewEnumerator2
    Sprite Size=&RandomRangeVector2D(Minimum=<5 100>, Maximum=<10 50>)
    Sprite Size Mode=::NewEnumerator4
    UsePositionOffset=false
    Write Material Random=true
   )
   &PartitionParticles(Number Of Partitions=$Emitter.PortCount)
   &SkeletalMeshLocation(
    Bone / Socket Index=$Particles.PartitionParticles.Partition
    Bone / Socket Sampling Mode=::NewEnumerator5
    MaskSpawnGroup=false
    Mesh Sampling Type=::NewEnumerator6
    Skeletal Mesh=$System.SkeletalMesh
   )
   &AddVelocity(
    Cone Angle=70
    Module.UseConeFalloff=false
    Module.UseVelDistribution=false
    Override Seed=false
    Velocity Mode=::NewEnumerator2
    Velocity Speed=&VC_FloatBySpawnGroup(
     SpawnGroup or ExecIndex:NiagaraInt32=0
     CustomInt:NiagaraInt32=0
     MaximumSpawnGroups:NiagaraInt32=2
     Float SG 0:NiagaraFloat=0
     Float SG 1:NiagaraFloat=-1000
     Float SG 2:NiagaraFloat=0
     Float SG 3:NiagaraFloat=0
     Float SG 4:NiagaraFloat=0
     Float SG 5:NiagaraFloat=0
    )
   )
   &InitialMeshOrientation(
    Mesh Orientation Mode=::NewEnumerator2
    Rotation=&RandomRangeVector(Minimum=<-1 0 0>, Maximum=<1 0 0>)
    Use Rotation Vector=true
   )
  ]
  particleUpdateScript [
   &ParticleState()
   &ScaleMeshSize(Scale Factor=&VectorFromCurve(
     VectorCurve=DataInterface() {propertyValues {
       XCurve {defaultValue=1, keys [<at 0 0 RCIM_Cubic>,<at 0.3 1 RCIM_Cubic>,<at 1 0 RCIM_Cubic>]}
       YCurve {defaultValue=1, keys [<at 0 0 RCIM_Cubic>,<at 0.3 1 RCIM_Cubic>,<at 1 0 RCIM_Cubic>]}
       ZCurve {keys [<at 0 0 RCIM_Cubic>,<at 0.3 1.7 RCIM_Cubic>,<at 1 0 RCIM_Cubic>]}
      }}
     Scale Curve=&RandomRangeVector(Minimum=<0.7 0.7 0.5>, Maximum=<1 1 1>)
    )) {enabled=false}
   &ScaleMeshSize(Scale Factor=&VectorFromCurve(
     VectorCurve=DataInterface() {propertyValues {
       XCurve {defaultValue=1, keys [<at 0 0 RCIM_Cubic>,<at 1 1.3 RCIM_Cubic>]}
       YCurve {defaultValue=1, keys [<at 0 0 RCIM_Cubic>,<at 1 1.3 RCIM_Cubic>]}
       ZCurve {keys [<at 0 0 RCIM_Cubic>,<at 1 1.7 RCIM_Cubic>]}
      }}
     Scale Curve=&RandomRangeVector(Minimum=<0.7 0.7 0.5>, Maximum=<1 1 1>)
    )) {module=ScaleMeshSize001}
   &Drag(
    Drag=&VC_FloatBySpawnGroup(
     SpawnGroup or ExecIndex:NiagaraInt32=0
     CustomInt:NiagaraInt32=0
     MaximumSpawnGroups:NiagaraInt32=3
     Float SG 0:NiagaraFloat=0
     Float SG 1:NiagaraFloat=1
     Float SG 2:NiagaraFloat=1
     Float SG 3:NiagaraFloat=0
     Float SG 4:NiagaraFloat=0
     Float SG 5:NiagaraFloat=0
    )
    Use Linear Drag=true
    Use Rotational Drag=true
   )
   &SolveForcesAndVelocity(Clamp Velocity=false, Limit Acceleration=false)
   &ScaleColor(
    Scale Alpha=&FloatFromCurve(FloatCurve=DataInterface() {propertyValues {Curve {keys [
         <at 0 0 RCIM_Cubic>
         {arriveTangentWeight=0.1067, interpMode=RCIM_Cubic, leaveTangentWeight=0.3202, tangentMode=RCTM_User, time=0.2, value=1}
         {arriveTangentWeight=0.3202, interpMode=RCIM_Cubic, leaveTangentWeight=0.1067, tangentMode=RCTM_User, time=0.8, value=1}
         <at 1 0 RCIM_Cubic>
        ]}}})
    Scale RGB=&VectorFromFloat(Value=&FloatFromCurve(FloatCurve=DataInterface() {propertyValues {Curve {keys [<at 0 10 RCIM_Cubic>,<at 1 1 RCIM_Cubic>]}}}))
    ScaleA=true
    ScaleRGB=true
   )
   &DynamicMaterialParameters(
    Index 0 Param 1=1
    Index 0 Param 3=&RandomRangeFloat(Minimum=$User.MeshFloat_Min, Maximum=$User.MeshFloat_Max)
    Index 0 Param 4=1
    Module.Index0Param0Write=true
    Module.Index0Param1Write=true
    Module.Index0Param2Write=true
    Module.Index0Param3Write=true
   ) {enabled=false}
   /SetVariables(Particles.SpriteColor2:LinearColor=&VC_DynamicScaleColor(
     Color to Scale:LinearColor=$Particles.R1 Color
     RGB Scale:Vector3f=&VectorFromFloat(Value=1)
     Alpha Scale:NiagaraFloat=&FloatFromCurve(FloatCurve=DataInterface() {propertyValues {Curve {
         defaultValue=1
         keys [
          <at 0 0 RCIM_Cubic>
          {arriveTangentWeight=0.0373, interpMode=RCIM_Cubic, leaveTangentWeight=0.3354, tangentMode=RCTM_User, time=0.2, value=1}
          {arriveTangentWeight=0.2178, interpMode=RCIM_Cubic, leaveTangentWeight=0.242, tangentMode=RCTM_User, time=0.5, value=1}
          {tangentMode=RCTM_User, time=1, value=0}
         ]
        }}})
    )) {isSetParameters=true}
  ]
  rendererClasses [&NiagaraMeshRendererProperties, &NiagaraSpriteRendererProperties]
  renderers [
   {rendererClass=&NiagaraMeshRendererProperties, rendererIndex=0}
   {rendererClass=&NiagaraSpriteRendererProperties, rendererIndex=1}
   {rendererClass=&NiagaraSpriteRendererProperties, rendererIndex=2}
  ]
 }
 Ribbon002 {
  emitterUpdateScript [
   &EmitterState()
   &VC_FilteredSocketCount()
   &VC_SpawnRibbon_v1(Number of Ribbons=$Emitter.PortCount, SpawnRate=60)
  ]
  particleSpawnScript [
   &VC_PaintedVariantsOverride(
    Black=<rgba 0 0 0 1>
    BurntSienna=<rgba 0.0642 0.0162 0.0033 1>
    Crimson=<rgba 0.1719 0 0.0047 1>
    DefaultColor=<rgba 0 0 0 1>
    ForestGreen=<rgba 0.0087 0.2344 0.0055 1>
    Grey=<rgba 0.0312 0.0312 0.0312 1>
    Lime=<rgba 0.1797 0.3594 0 1>
    Orange=<rgba 0.4115 0.1435 0 1>
    Pink=<rgba 0.3229 0.0858 0.2367 1>
    Purple=<rgba 0.1026 0 0.25 1>
    Saffron=<rgba 0.4844 0.4102 0.0304 1>
    SkyBlue=<rgba 0 0.1941 0.2917 1>
    TitaniumWhite=<rgba 0.4219 0.4219 0.4219 1>
   )
   &VC_InitializeRibbon_v1(Color=$Particles.Color, Enable Skeletal Position=true, Ribbon Width=55)
   &AddVelocity(Velocity=<-1500 0 0>, Velocity Speed Scale=2) {module=AddVelocity001}
  ]
  particleUpdateScript [
   &ParticleState()
   &SolveForcesAndVelocity(Clamp Velocity=false, Limit Acceleration=false)
   &VC_UpdateRibbon_v1(Enable Skeletal Position=true)
   &ScaleColor(
    Scale Alpha=&FloatFromCurve(FloatCurve=DataInterface() {propertyValues {Curve {keys [
         <at 0 0 RCIM_Cubic>
         {arriveTangentWeight=0.2125, interpMode=RCIM_Cubic, leaveTangentWeight=0.3542, tangentMode=RCTM_User, time=0.1, value=1}
         {arriveTangentWeight=0.3941, interpMode=RCIM_Cubic, leaveTangentWeight=0.1576, tangentMode=RCTM_User, time=0.5, value=1}
        ]}}})
    Scale RGB=&VectorFromFloat(Value=&FloatFromCurve(FloatCurve=DataInterface() {propertyValues {Curve {keys [
          <at 0 0.3 RCIM_Cubic>
          {arriveTangentWeight=0.2125, interpMode=RCIM_Cubic, leaveTangentWeight=0.3542, tangentMode=RCTM_User, time=0.2, value=1}
          {arriveTangentWeight=0.3941, interpMode=RCIM_Cubic, leaveTangentWeight=0.1576, tangentMode=RCTM_User, time=0.8, value=1}
          <at 1 0.3 RCIM_Cubic>
         ]}}}))
    ScaleA=true
    ScaleRGB=true
   )
   &ScaleRibbonWidth(Ribbon Width Scale=&FloatFromCurve(FloatCurve=DataInterface() {propertyValues {Curve {keys [{arriveTangent=0.054, interpMode=RCIM_Cubic, leaveTangent=0.054, tangentMode=RCTM_User, time=0.4, value=1}, <at 1 0.1 RCIM_Cubic>]}}}))
  ]
  rendererClasses [&NiagaraRibbonRendererProperties]
  renderers [{rendererClass=&NiagaraRibbonRendererProperties}]
 }
 Ribbon_1_CW {
  emitterUpdateScript [
   &EmitterState()
   &VC_FilteredSocketCount()
   &VC_RateQualityHandler()
   &VC_SpawnRibbon_v1(Number of Ribbons=$Emitter.PortCount, SpawnRate=60)
  ]
  particleSpawnScript [
   &VC_PaintedVariantsOverride(
    Black=<rgba 0.002 0.002 0.002 1>
    BurntSienna=<rgba 0.05 0.01 0.0008 1>
    Cobalt=<rgba 0 0.0555 3 1>
    Crimson=<rgba 0.1458 0 0.004 1>
    DefaultColor=<rgba 2 0.1391 0 1>
    ForestGreen=<rgba 0 1 0.0505 1>
    Grey=<rgba 0.0208 0.0208 0.0208 1>
    Lime=<rgba 0.2943 0.5885 0 1>
    Pink=<rgba 1.5156 0.0632 0.9875 1>
    Purple=<rgba 0.3968 0 0.8906 1>
    Saffron=<rgba 0.7135 0.547 0 1>
   )
   &VC_InitializeRibbon_v1(Color=$Particles.Color, Enable Skeletal Position=true, Lifetime=1, Ribbon Width=15)
   &AddVelocity(Velocity=&MakeVector(
     X=-2000
     Y=&Sine(Normalized Angle=$Emitter.Age, Period=0.1, Scale=40, Bias=0)
     Z=&Cosine(Normalized Angle=$Emitter.Age, Period=0.1, Scale=40, Bias=0)
    )) {module=AddVelocity001}
   &ShapeLocation(
    MaskSpawnGroup=false
    Offset Mode=::NewEnumerator0
    Override Seed=false
    Radius Position=1.2
    Ring / Disc Distribution Mode:ENiagaraSphereDistributionMode=::NewEnumerator1
    Ring Radius=-30
    Rotation Angle Type:ENiagara_AngleInput=::NewEnumerator2
    Rotation Coordinate Space=::Simulation
    Rotation Mode=::NewEnumerator2
    Shape Primitive=::NewEnumerator4
    UseOwnerScale=true
    Yaw / Pitch / Roll=&Multiply_VectorByFloat(Vector=<360 0 360>, Float=1)
   )
  ]
  particleUpdateScript [
   &ParticleState()
   &SolveForcesAndVelocity(Clamp Velocity=false, Limit Acceleration=false)
   &VC_UpdateRibbon_v1(Enable Skeletal Position=true)
   &ScaleRibbonWidth(Ribbon Width Scale=&FloatFromCurve(FloatCurve=DataInterface() {propertyValues {Curve {keys [
         <at 0 0 RCIM_Cubic>
         {arriveTangentWeight=0.1067, interpMode=RCIM_Cubic, leaveTangentWeight=0.3202, tangentMode=RCTM_User, time=0.2, value=1}
         {arriveTangentWeight=0.3202, interpMode=RCIM_Cubic, leaveTangentWeight=0.1067, tangentMode=RCTM_User, time=0.8, value=1}
         <at 1 0 RCIM_Cubic>
        ]}}}))
   &ScaleColor(
    Scale Alpha=&FloatFromCurve(FloatCurve=DataInterface() {propertyValues {Curve {keys [<at 0.8 1>,<at 1 0>]}}})
    Scale RGB=&VectorFromFloat(Value=4)
    ScaleA=true
    ScaleRGB=true
   )
   &DynamicMaterialParameters(
    Index 0 Param 1=&MakeCustomFloatFromBool(Bool=$User.bIsPainted, True Float=1, False Float=1)
    Module.Index0Float0Write=true
    Module.Index0Float1Write=false
    Module.Index0Float2Write=true
    Module.Index0Float3Write=false
   )
  ]
  rendererClasses [&NiagaraRibbonRendererProperties]
  renderers [
   {rendererClass=&NiagaraRibbonRendererProperties}
   {rendererClass=&NiagaraRibbonRendererProperties, rendererIndex=1}
  ]
 }
 Ribbon_1_CW001 {
  emitterUpdateScript [
   &EmitterState()
   &VC_FilteredSocketCount()
   &VC_RateQualityHandler()
   &VC_SpawnRibbon_v1(Number of Ribbons=$Emitter.PortCount, SpawnRate=60)
  ]
  particleSpawnScript [
   &VC_PaintedVariantsOverride(
    Black=<rgba 0.002 0.002 0.002 1>
    BurntSienna=<rgba 0.05 0.01 0.0008 1>
    Cobalt=<rgba 0 0.0555 3 1>
    Crimson=<rgba 0.1458 0 0.004 1>
    DefaultColor=<rgba 2 0.1391 0 1>
    ForestGreen=<rgba 0 1 0.0505 1>
    Grey=<rgba 0.0208 0.0208 0.0208 1>
    Pink=<rgba 1.5156 0.0632 0.9875 1>
    Purple=<rgba 0.3968 0 0.8906 1>
    Saffron=<rgba 0.7135 0.547 0 1>
   )
   &VC_InitializeRibbon_v1(Color=$Particles.Color, Enable Skeletal Position=true, Lifetime=1, Ribbon Width=15)
   &AddVelocity(Velocity=&MakeVector(
     X=-2000
     Y=&Sine(Normalized Angle=$Emitter.Age, Period=0.1, Scale=40, Bias=0)
     Z=&Cosine(Normalized Angle=$Emitter.Age, Period=0.1, Scale=40, Bias=0)
    )) {module=AddVelocity001}
   &ShapeLocation(
    MaskSpawnGroup=false
    Offset Mode=::NewEnumerator0
    Override Seed=false
    Radius Position=1.2
    Ring / Disc Distribution Mode:ENiagaraSphereDistributionMode=::NewEnumerator1
    Ring Radius=-30
    Rotation Angle Type:ENiagara_AngleInput=::NewEnumerator2
    Rotation Coordinate Space=::Simulation
    Rotation Mode=::NewEnumerator2
    Shape Primitive=::NewEnumerator4
    UseOwnerScale=true
    Yaw / Pitch / Roll=&Multiply_VectorByFloat(Vector=<-360 0 -360>, Float=1)
   )
  ]
  particleUpdateScript [
   &ParticleState()
   &SolveForcesAndVelocity(Clamp Velocity=false, Limit Acceleration=false)
   &VC_UpdateRibbon_v1(Enable Skeletal Position=true)
   &ScaleRibbonWidth(Ribbon Width Scale=&FloatFromCurve(FloatCurve=DataInterface() {propertyValues {Curve {keys [
         <at 0 0 RCIM_Cubic>
         {arriveTangentWeight=0.1067, interpMode=RCIM_Cubic, leaveTangentWeight=0.3202, tangentMode=RCTM_User, time=0.2, value=1}
         {arriveTangentWeight=0.3202, interpMode=RCIM_Cubic, leaveTangentWeight=0.1067, tangentMode=RCTM_User, time=0.8, value=1}
         <at 1 0 RCIM_Cubic>
        ]}}}))
   &ScaleColor(
    Scale Alpha=&FloatFromCurve(FloatCurve=DataInterface() {propertyValues {Curve {keys [<at 0 0>,<at 0.3 1>,<at 0.7 1>,<at 1 0>]}}})
    Scale RGB=&VectorFromFloat(Value=&FloatFromCurve(FloatCurve=DataInterface() {propertyValues {Curve {keys [<at 0 0>,<at 0.3 1>,<at 0.7 1>,<at 1 0>]}}}, Scale Curve=4))
    ScaleA=true
    ScaleRGB=true
   )
   &DynamicMaterialParameters(
    Index 0 Param 1=&MakeCustomFloatFromBool(Bool=$User.bIsPainted, True Float=1, False Float=1)
    Module.Index0Float0Write=true
    Module.Index0Float1Write=false
    Module.Index0Float2Write=true
    Module.Index0Float3Write=false
   )
  ]
  rendererClasses [&NiagaraRibbonRendererProperties]
  renderers [
   {rendererClass=&NiagaraRibbonRendererProperties}
   {rendererClass=&NiagaraRibbonRendererProperties, rendererIndex=1}
  ]
 }
 Sprites {
  emitterUpdateScript [
   &EmitterState()
   &VC_FilteredSocketCount()
   &SpawnRate(SpawnRate=20, Use Spawn Probability=false)
   &SpawnRate(SpawnGroup=1, SpawnRate=20, Use Spawn Probability=false) {module=SpawnRate001}
   &SpawnRate(SpawnGroup=2, SpawnRate=80, Use Spawn Probability=false) {module=SpawnRate002}
  ]
  particleSpawnScript [
   /SetVariables(Particles.RenderTag:NiagaraInt32=&SpawnGroup()) {isSetParameters=true}
   &VC_PaintedVariantsOverride(
    BurntSienna=<rgba 0.4635 0.0981 0 1>
    Cobalt=<rgba 0.0435 0.0925 5 1>
    DefaultColor=<rgba 3.7462 0.2024 0 1>
    R1 Black:LinearColor=<rgba 0 0 0 1>
    R1 BurntSienna:LinearColor=<rgba 0 0 0 1>
    R1 Cobalt:LinearColor=<rgba 0 0 0 1>
    R1 Crimson:LinearColor=<rgba 0 0 0 1>
    R1 DefaultColor=<rgba 0 0 0 1>
    R1 ForestGreen:LinearColor=<rgba 0 0 0 1>
    R1 Grey:LinearColor=<rgba 0 0 0 1>
    R1 Lime:LinearColor=<rgba 0 0 0 1>
    R1 Orange:LinearColor=<rgba 0 0 0 1>
    R1 Pink:LinearColor=<rgba 0 0 0 1>
    R1 Purple:LinearColor=<rgba 0 0 0 1>
    R1 Saffron:LinearColor=<rgba 0 0 0 1>
    R1 SkyBlue:LinearColor=<rgba 0 0 0 1>
    R1 TitaniumWhite:LinearColor=<rgba 0.3125 0.3125 0.3125 1>
    SkyBlue=<rgba 0 0.357 0.5365 1>
    Use R1 Painted Overrides=true
   )
   &InitializeParticle(
    Color=$Particles.Color
    Lifetime=&VC_FloatBySpawnGroup(
     SpawnGroup or ExecIndex:NiagaraInt32=0
     CustomInt:NiagaraInt32=0
     MaximumSpawnGroups:NiagaraInt32=3
     Float SG 0:NiagaraFloat=&RandomRangeFloat(Minimum=0.8, Maximum=0.5)
     Float SG 1:NiagaraFloat=&RandomRangeFloat(Minimum=0.8, Maximum=0.5)
     Float SG 2:NiagaraFloat=&RandomRangeFloat(Minimum=1.5, Maximum=1)
     Float SG 3:NiagaraFloat=0
     Float SG 4:NiagaraFloat=0
     Float SG 5:NiagaraFloat=0
    )
    Sprite Rotation Mode=::NewEnumerator2
    Sprite Size=&RandomRangeVector2D(Minimum=<10 50>, Maximum=<5 80>)
    Sprite Size Mode=::NewEnumerator4
    UsePositionOffset=false
    Write Material Random=true
   )
   &PartitionParticles(Number Of Partitions=$Emitter.PortCount)
   &SkeletalMeshLocation(
    Bone / Socket Index=$Particles.PartitionParticles.Partition
    Bone / Socket Sampling Mode=::NewEnumerator5
    MaskSpawnGroup=false
    Mesh Sampling Type=::NewEnumerator6
    Skeletal Mesh=$System.SkeletalMesh
   )
   &AddVelocity(
    Velocity=<-1200 0 0>
    Velocity Speed Scale=&VC_FloatBySpawnGroup(
     SpawnGroup or ExecIndex:NiagaraInt32=0
     CustomInt:NiagaraInt32=0
     MaximumSpawnGroups:NiagaraInt32=3
     Float SG 0:NiagaraFloat=1
     Float SG 1:NiagaraFloat=1
     Float SG 2:NiagaraFloat=0.5
     Float SG 3:NiagaraFloat=0
     Float SG 4:NiagaraFloat=0
     Float SG 5:NiagaraFloat=0
    )
   )
   &ShapeLocation(
    Cylinder Radius=40
    MaskSpawnGroup=false
    Offset Mode=::NewEnumerator0
    Override Seed=false
    Shape Primitive=::NewEnumerator1
    Surface Only=false
    UseOwnerScale=true
   )
   /SetVariables(Particles.SizeTag2:Vector2f=&SelectVector2DFromArray(
     Array Sampling Mode=::NewEnumerator0
     Direct Array Index=?
     Vector 2D Selection Array=DataInterface() {propertyValues {FloatData [<15 15>, <25 25>]}}
     Random Seed=0
    )) {isSetParameters=true}
  ]
  particleUpdateScript [
   &ParticleState()
   &ScaleSpriteSize(Uniform Curve Sprite Scale=DataInterface() {propertyValues {Curve {keys [
        <at 0 0.1 RCIM_Cubic>
        {arriveTangent=-0.1, interpMode=RCIM_Cubic, leaveTangent=-0.1, time=0.1, value=1.8}
        {tangentMode=RCTM_User, time=1, value=0}
       ]}}})
   &ScaleColor(Scale RGB=&VectorFromFloat(Value=3), ScaleA=true, ScaleRGB=true)
   &InheritSourceMovement(
    Applied Position Scale=&VectorFromFloat(Value=&FloatFromCurve(FloatCurve=DataInterface() {propertyValues {Curve {keys [<at 0 1>,<at 1 0>]}}}))
    ApplyForce=false
    ApplyPosition=true
    ApplyPositionFalloff=false
   )
   &SolveForcesAndVelocity(Clamp Velocity=false, Limit Acceleration=false)
   &RotateAroundPoint(
    Radius:NiagaraFloat=&VC_FloatBySpawnGroup(
     SpawnGroup or ExecIndex:NiagaraInt32=0
     CustomInt:NiagaraInt32=0
     MaximumSpawnGroups:NiagaraInt32=3
     Float SG 0:NiagaraFloat=0
     Float SG 1:NiagaraFloat=0
     Float SG 2:NiagaraFloat=&RandomRangeFloat(Minimum=-10, Maximum=10)
     Float SG 3:NiagaraFloat=0
     Float SG 4:NiagaraFloat=0
     Float SG 5:NiagaraFloat=0
    )
    Rotation Center:NiagaraPosition=$Particles.Position
    Rotation Rate:NiagaraFloat=&VC_FloatBySpawnGroup(
     SpawnGroup or ExecIndex:NiagaraInt32=0
     CustomInt:NiagaraInt32=0
     MaximumSpawnGroups:NiagaraInt32=3
     Float SG 0:NiagaraFloat=0
     Float SG 1:NiagaraFloat=0
     Float SG 2:NiagaraFloat=&Multiply_Float(
      A=2
      B=&SelectFloatFromArray(Direct Array Index=0, Float Selection Array=DataInterface() {propertyValues {FloatData [-2,5,3,-3]}})
     )
     Float SG 3:NiagaraFloat=0
     Float SG 4:NiagaraFloat=0
     Float SG 5:NiagaraFloat=0
    )
    X Axis:Vector3f=<360 360 360>
    Y Axis:Vector3f=<-360 -360 -360>
   )
   &DynamicMaterialParameters(
    Index 0 Param 1=&FloatFromCurve(FloatCurve=DataInterface() {propertyValues {Curve {keys [
         {arriveTangent=0.0795, interpMode=RCIM_Cubic, leaveTangent=0.0795, tangentMode=RCTM_User, time=0, value=0}
         {arriveTangent=0.0414, interpMode=RCIM_Cubic, leaveTangent=0.0414, tangentMode=RCTM_User, time=0.7, value=0.05}
         {tangentMode=RCTM_User, time=1, value=0.5}
        ]}}})
    Module.Index0Float0Write=true
    Module.Index0Float1Write=true
    Module.Index0Float2Write=true
    Module.Index0Float3Write=true
   )
   /SetVariables(
    Particles.SpriteColor1:LinearColor=&VC_DynamicScaleColor(
     Color to Scale:LinearColor=$Particles.Initial.R1 Color
     RGB Scale:Vector3f=&VectorFromFloat(Value=1)
     Alpha Scale:NiagaraFloat=3
    )
    Particles.SpriteSize2=&Multiply_Vector2D_ByFloat(
     Float=&FloatFromCurve(FloatCurve=DataInterface() {propertyValues {Curve {keys [
          {arriveTangent=-1.9328, interpMode=RCIM_Cubic, leaveTangent=-1.9328, tangentMode=RCTM_User, time=0, value=0}
          {arriveTangentWeight=0.119, interpMode=RCIM_Cubic, leaveTangentWeight=1.0713, tangentMode=RCTM_User, time=0.1, value=1}
          {arriveTangentWeight=0.299, interpMode=RCIM_Cubic, leaveTangentWeight=0.1495, tangentMode=RCTM_User, time=0.7, value=1}
          {arriveTangent=-0.0405, interpMode=RCIM_Cubic, leaveTangent=-0.0405, tangentMode=RCTM_User, time=1, value=0}
         ]}}})
     Vector2D=$Particles.Initial.SizeTag2
    )
   ) {isSetParameters=true}
  ]
  rendererClasses [&NiagaraSpriteRendererProperties]
  renderers [
   {rendererClass=&NiagaraSpriteRendererProperties}
   {rendererClass=&NiagaraSpriteRendererProperties, rendererIndex=1}
   {rendererClass=&NiagaraSpriteRendererProperties, rendererIndex=2}
  ]
 }
}
summary {
 systemName=NS_Boost_Illustrative
 emitters [
  {emitterName=ConeBase, rendererClasses [&NiagaraMeshRendererProperties, &NiagaraSpriteRendererProperties]}
  {emitterName=Sprites, rendererClasses [&NiagaraSpriteRendererProperties]}
  {emitterName=Ribbon002, rendererClasses [&NiagaraRibbonRendererProperties]}
  {emitterName=Ribbon_1_CW, rendererClasses [&NiagaraRibbonRendererProperties]}
  {emitterName=Ribbon_1_CW001, rendererClasses [&NiagaraRibbonRendererProperties]}
 ]
 userVariables [
  User.bIsFrontendPreview {type=NiagaraBool}
  User.bIsPainted {type=NiagaraBool}
  User.blackRibbon {type=LinearColor, description="", defaultValue {struct=LinearColor, value=<rgba 1 1 1 1>}}
  User.BoostMesh {type=LinearColor, description="", defaultValue {struct=LinearColor, value=<rgba 1 1 1 1>}}
  User.BoostSprite {type=LinearColor, description="", defaultValue {struct=LinearColor, value=<rgba 1 1 1 1>}}
  User.BoostWipe_1 {type=LinearColor, description="", defaultValue {struct=LinearColor, value=<rgba 1 1 1 1>}}
  User.BoostWipe_ 2 {type=LinearColor, description="", defaultValue {struct=LinearColor, value=<rgba 1 1 1 1>}}
  User.ColorVariantID {type=NiagaraInt32}
  User.dynamicBlackRibbon {type=NiagaraFloat, defaultValue=0, description=""}
  User.DynamicRibbonAddColor {type=NiagaraFloat, defaultValue=0, description=""}
  User.DynamicRibbonTwist_ {type=NiagaraFloat, defaultValue=1, description=""}
  User.EmissiveMeshBoost {type=NiagaraFloat, defaultValue=1, description=""}
  User.MeshFloat_Max {type=NiagaraFloat, defaultValue=-1, description=""}
  User.MeshFloat_Min {type=NiagaraFloat, defaultValue=0.1, description=""}
  User.PaintedColor {type=LinearColor, description="", defaultValue {struct=LinearColor, value=<rgba 1 1 1 1>}}
  User.PaintedScalar {type=NiagaraFloat, defaultValue=1, description=""}
  User.Particles {type=LinearColor, description="", defaultValue {struct=LinearColor, value=<rgba 1 1 1 1>}}
  User.PortCount {type=NiagaraInt32, defaultValue=1}
  User.RibbonTwistColor {type=LinearColor, description="", defaultValue {struct=LinearColor, value=<rgba 1 1 1 1>}}
 ]
}
