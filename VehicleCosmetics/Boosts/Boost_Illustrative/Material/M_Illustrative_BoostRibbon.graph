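# Translucent, unlit, two-sided Niagara ribbon material for the Illustrative vehicle boost cosmetic.
# A tiled, time-panned texture is power-shaped to index a color LUT, then tinted by particle color and boosted by dynamic parameters into clamped emissive.
# Texture alpha times particle alpha forms the opacity signal.
# Static switches add thumbnail desaturation and optional sine/cosine spiral world position offset, with selectable spin direction and masking along the second UV channel.
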Material /VehicleCosmetics/Boosts/Boost_Illustrative/Material/M_Illustrative_BoostRibbon
graph {
 properties {
  bUsedWithNiagaraRibbons=true
  blendMode=BLEND_Translucent
  pixelDepthOffsetMode=PDOM_Legacy
  shadingModel=MSM_Unlit
  twoSided=true
  usedShadingModels=Unlit
 }
 nodes [
  n1 = DynamicParameter() {paramNames [EmissiveBoost, Param2, RibbonColorAdd, Param4]}
  n2 = VectorParameter() {parameterName=BloomMult, defaultValue=<rgba 1 1 1 1>}
  n3 = ParticleColor()
  n4 = ScalarParameter() {defaultValue=1.5, parameterName=Multiplier}
  n5 = TextureCoordinate()
  n6 = AppendVector(A=ScalarParameter() {defaultValue=1.5, parameterName=Tile U}, B=ScalarParameter() {defaultValue=1, parameterName=Tile V})
  n7 = Multiply(A=Time(), B=ScalarParameter() {defaultValue=1, parameterName=TimeMulti})
  n8 = ScalarParameter() {defaultValue=2, parameterName=TimePeriod}
  n9 = TextureSampleParameter2D(UVs=Panner(Coordinate=Multiply(A=@n5, B=@n6), Time=Fmod(A=@n7, B=@n8)) {speedX=-2}) {parameterName=Main_Text, texture=/VehicleCosmetics/Boosts/Boost_Illustrative/Textures/T_VCVFX_Boost_Illustrative_001_Ribbon02_Alpha_TX_RGB}
  n10 = Power(Base=Saturate(None=@n9.R), Exp=ScalarParameter() {defaultValue=0.09, parameterName=PowerExp1})
  n11 = TextureSampleParameter2D(UVs=AppendVector(A=@n10, B=Constant())) {parameterName=LUT Texture, texture=/VehicleCosmetics/Boosts/Boost_Illustrative/Textures/T_VCVFX_Boost_Illustrative_001_LUT_TX_Diffuse}
  n12 = Multiply(
   A=Power(Base=Multiply(A=@n4, B=@n10), Exp=ScalarParameter() {defaultValue=3.5, parameterName=PowerExp2})
   B=Add(A=@n11.RGB, B=VectorParameter() {parameterName=Add Color, defaultValue=<rgba 0.05 0.05 0.05 1>}.RGB)
  )
  n13 = Desaturation(None=@n12)
  n14 = Multiply(A=@n3.RGB, B=@n13)
  n15 = Add(A=@n1.RibbonColorAdd, B=Multiply(A=@n1.EmissiveBoost, B=Multiply(A=@n2.RGB, B=@n14)))
  MaterialFunctionCall(Emissive=Multiply(A=Constant() {r=5}, B=Clamp(None=@n15) {maxDefault=1000}) {constA=5}) {materialFunction=&MF_FX_EmissiveAdjustments}
  n16 = DynamicParameter() {paramNames [EmissiveBoost, Param2, RibbonColorAdd, Param4]}
  n17 = ParticleColor()
  n18 = ScalarParameter() {defaultValue=1.5, parameterName=Multiplier}
  n19 = TextureCoordinate()
  n20 = AppendVector(A=ScalarParameter() {defaultValue=1.5, parameterName=Tile U}, B=ScalarParameter() {defaultValue=1, parameterName=Tile V})
  n21 = Multiply(A=Time(), B=ScalarParameter() {defaultValue=1, parameterName=TimeMulti})
  n22 = ScalarParameter() {defaultValue=2, parameterName=TimePeriod}
  n23 = TextureSampleParameter2D(UVs=Panner(Coordinate=Multiply(A=@n19, B=@n20), Time=Fmod(A=@n21, B=@n22)) {speedX=-2}) {parameterName=Main_Text, texture=&T_VCVFX_Boost_Illustrative_001_Ribbon02_Alpha_TX_RGB}
  n24 = Power(Base=Saturate(None=@n23.R), Exp=ScalarParameter() {defaultValue=0.09, parameterName=PowerExp1})
  n25 = TextureSampleParameter2D(UVs=AppendVector(A=@n24, B=Constant())) {parameterName=LUT Texture, texture=&T_VCVFX_Boost_Illustrative_001_LUT_TX_Diffuse}
  n26 = Multiply(
   A=Power(Base=Multiply(A=@n18, B=@n24), Exp=ScalarParameter() {defaultValue=3.5, parameterName=PowerExp2})
   B=Add(A=@n25.RGB, B=VectorParameter() {parameterName=Add Color, defaultValue=<rgba 0.05 0.05 0.05 1>}.RGB)
  )
  n27 = StaticSwitchParameter(True=Desaturation(None=@n26), False=@n26) {parameterName=DesaturateForThumbnail}
  n28 = Multiply(
   A=@n16.EmissiveBoost
   B=Multiply(A=VectorParameter() {parameterName=BloomMult, defaultValue=<rgba 1 1 1 1>}.RGB, B=Multiply(A=@n17.RGB, B=@n27))
  )
  Multiply(A=Constant() {r=5}, B=Clamp(None=Add(A=@n16.RibbonColorAdd, B=@n28)) {maxDefault=1000}) {constA=5}
  Multiply(A=@n17.A, B=@n23.A)
  Multiply(A=@n3.A, B=@n9.A)
  StaticSwitchParameter(True=@n13, False=@n12) {parameterName=DesaturateForThumbnail}
  n29 = TextureCoordinate()
  n30 = ScalarParameter() {defaultValue=1, parameterName=SinUVTile}
  n31 = Multiply(A=ScalarParameter() {defaultValue=1, parameterName=SinUVTimeMult}, B=Time())
  n32 = ScalarParameter() {parameterName=SinUVTimeOffset}
  n33 = Add(A=ComponentMask(None=Multiply(A=@n29, B=@n30)) {r=true}, B=Fmod(A=Add(A=@n31, B=@n32), B=Constant() {r=1}))
  n34 = Sine(None=@n33)
  n35 = Cosine(None=@n33)
  n36 = AppendVector(
   A=AppendVector(A=Constant(), B=StaticSwitchParameter(True=@n34, False=@n35) {parameterName=Spin_CCW})
   B=StaticSwitchParameter(True=@n35, False=@n34) {parameterName=Spin_CCW}
  )
  n37 = ScalarParameter() {defaultValue=1, parameterName=WPO Mult}
  n38 = Power(
   Base=ComponentMask(None=TextureCoordinate() {coordinateIndex=1}) {r=true}
   Exp=ScalarParameter() {defaultValue=0.5, parameterName=WPO_Mask_Power}
  )
  n39 = ScalarParameter() {defaultValue=4, parameterName=WPO_Mask_Mult}
  n40 = Saturate(None=Add(A=ScalarParameter() {defaultValue=0.25, parameterName=WPO_Mask_Min}, B=Multiply(A=@n38, B=@n39)))
  n41 = Constant() {r=1}
  StaticSwitchParameter(True=Multiply(A=Multiply(A=@n36, B=@n37), B=StaticSwitchParameter(True=@n40, False=@n41) {parameterName=Mask_WPO}), False=Constant()) {parameterName=Use_WPO}
  n42 = TextureCoordinate()
  n43 = ScalarParameter() {defaultValue=1, parameterName=SinUVTile}
  n44 = Multiply(A=ScalarParameter() {defaultValue=1, parameterName=SinUVTimeMult}, B=Time())
  n45 = ScalarParameter() {parameterName=SinUVTimeOffset}
  n46 = Add(A=ComponentMask(None=Multiply(A=@n42, B=@n43)) {r=true}, B=Fmod(A=Add(A=@n44, B=@n45), B=Constant() {r=1}))
  n47 = Sine(None=@n46)
  n48 = Cosine(None=@n46)
  n49 = AppendVector(
   A=AppendVector(A=Constant(), B=StaticSwitchParameter(True=@n47, False=@n48) {parameterName=Spin_CCW})
   B=StaticSwitchParameter(True=@n48, False=@n47) {parameterName=Spin_CCW}
  )
  n50 = ScalarParameter() {defaultValue=1, parameterName=WPO Mult}
  n51 = Power(
   Base=ComponentMask(None=TextureCoordinate() {coordinateIndex=1}) {r=true}
   Exp=ScalarParameter() {defaultValue=0.5, parameterName=WPO_Mask_Power}
  )
  n52 = ScalarParameter() {defaultValue=4, parameterName=WPO_Mask_Mult}
  n53 = Saturate(None=Add(A=ScalarParameter() {defaultValue=0.25, parameterName=WPO_Mask_Min}, B=Multiply(A=@n51, B=@n52)))
  n54 = Constant() {r=1}
  StaticSwitchParameter(True=Multiply(A=Multiply(A=@n49, B=@n50), B=StaticSwitchParameter(True=@n53, False=@n54) {parameterName=Mask_WPO}), False=Constant()) {parameterName=Use_WPO}
 ]
}
